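# UI material for the LastVoice Item Shop art, building an animated purple and pink cloudy sky background.
# Cloud texture distorted by grayscale flow maps and panning blurred noise combine with linear and top-bottom gradients into masks for background, middle and foreground clouds.
# Gradient-tinted layers are lerped over a sky base, then faded against dark navy into the BestBG reroute.
# Named reroutes link the stages, and some loose nodes, including an MF_UI_Clouds call, feed nothing.
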Material /ItemShop/UI/Art/LastVoice/M_UI_LastVoice_DN
graph {
 properties {bUsedWithStaticMesh=false, materialDomain=MD_UI, pixelDepthOffsetMode=PDOM_Legacy, usedShadingModels=Default Lit}
 nodes [
  n1 = AppendVector(A=Constant(), B=Constant() {r=2})
  n2 = NamedRerouteUsage()
  Add(A=@n1, B=@n2)
  Constant3Vector() {constant=<rgba 0.5457 0.4406 0.5349 1>}
  Constant3Vector() {constant=<rgba 0.5397 0.2349 0.9047 1>}
  Constant3Vector() {constant=<rgba 0.4219 0.0549 0.2275 1>}
  MaterialFunctionCall!(
   Mask
   CloudIntensity=Constant() {r=1}
   Density=Constant() {r=1}
   Scale=Constant() {r=1}
   Speed
   CloudColorLow
   CloudColorHigh=Constant3Vector() {constant=<rgba 0.8963 0.8796 0.9047 1>}
   CloudColorDark=Constant3Vector() {constant=<rgba 0.5457 0.1594 0.3718 1>}
   CloudDarkIntensity=Constant() {r=2}
   FoggyAlphaMin=Constant() {r=5}
   NoiseTexture=TextureObject() {samplerType=SAMPLERTYPE_Alpha, texture=/Game/UI/Frontend/ItemShop/Art/LastVoice/T_UI_CloudLV}
  ) {materialFunction=/FortUILibrary/MaterialFunctions/Effects/MF_UI_Clouds}
  n3 = Constant3Vector() {constant=<rgba 0.0203 0.0103 0.147 1>}
  MaterialFunctionCall(
   A=@n3
   B=Constant3Vector() {constant=<rgba 0.0382 0.2961 0.6939 1>}
   C=Constant3Vector() {constant=<rgba 0.8178 0.8459 0.9047 1>}
   Alpha=MaterialFunctionCall(GradientStart=Constant() {r=-0.1}, GradientEnd=Constant() {r=0.45}, RotationAngle=Constant() {r=0.72}, GradientOffset=Constant() {r=0.1}) {materialFunction=&MF_UI_LinearGradient}.Result
  ) {materialFunction=&3ColorBlend}
  n4 = NamedRerouteUsage()
  MaterialFunctionCall(UVs=@n4, Texture Scale=Constant() {r=0.8}) {materialFunction=&ScaleUVsByCenter}
  n5 = NamedRerouteUsage()
  n6 = Constant() {r=0.75}
  n7 = TextureSample(UVs=Panner(Coordinate=Multiply(A=@n5, B=@n6)) {speedX=0.01, speedY=0.01}) {samplerType=SAMPLERTYPE_Grayscale, texture=&LowResBlurredNoise}
  n8 = Multiply(A=@n7.R) {constB=1.5}
  n9 = Constant()
  n10 = Constant() {r=4}
  n11 = MaterialFunctionCall!(
   UVs=Multiply(A=AppendVector(A=@n9, B=@n10), B=NamedRerouteUsage())
   Subtract=Constant()
   Falloff=Constant() {r=0.1}
   Intensity=Constant() {r=2}
  ) {materialFunction=/Game/UI/Frontend/ItemShop/Art/_MaterialFunctions_ShopTheming/MF_UI_TopBottomGradient}
  n12 = NamedRerouteUsage()
  n13 = Multiply(A=OneMinus(None=SmoothStep(Value=@n11.Result)), B=Max(A=Multiply(A=@n12) {constB=1.25}, B=@n7.R))
  n14 = SmoothStep(Value=OneMinus(None=Add(A=@n8, B=@n13))) {constMax=0.35}
  n15 = Constant()
  n16 = Constant() {r=0.1}
  n17 = MaterialFunctionCall(
   UVs=Add(A=NamedRerouteUsage(), B=AppendVector(A=@n15, B=@n16))
   GradientStart=Constant() {r=-0.01}
   GradientEnd=Constant() {r=0.96}
   RotationAngle=Constant() {r=0.75}
   GradientOffset=Constant() {r=0.1}
  ) {materialFunction=&MF_UI_LinearGradient}
  n18 = Saturate(None=Multiply(A=@n14, B=Min(A=@n17.Result, B=@n14)))
  n19 = Constant3Vector() {constant=<rgba 0.107 0.0319 0.4125 1>}
  n20 = Constant() {r=0.15}
  n21 = Constant() {r=1}
  n22 = Constant() {r=0.75}
  n23 = Constant() {r=0.25}
  n24 = LinearInterpolate(
   A=MaterialFunctionCall(
    A=Constant3Vector() {constant=<rgba 0.8423 0.3128 0.9688 1>}
    B=Constant3Vector() {constant=<rgba 0.3046 0.0744 0.7031 1>}
    C=@n19
    Alpha=NamedRerouteUsage()
   ) {materialFunction=&3ColorBlend}.Result
   B=@n19
   Alpha=SmoothStep(Value=MaterialFunctionCall(GradientStart=@n20, GradientEnd=@n21, RotationAngle=@n22, GradientOffset=@n23) {materialFunction=&MF_UI_LinearGradient}.Result)
  )
  Multiply(A=@n18, B=@n24)
  NamedRerouteDeclaration(None=MaterialFunctionCall() {materialFunction=&MF_UI_UVScaling}.Scaled to Fit) {name=UVScaling_1, nodeColor=<rgba 0 0.9297 1 1>}
  n25 = Multiply(A=NamedRerouteUsage()) {constB=1.05}
  n26 = Panner(Coordinate=@n25) {speedX=0.01, speedY=0.001}
  n27 = TextureObject() {samplerType=SAMPLERTYPE_Alpha, texture=&T_UI_CloudLV}
  n28 = TextureSample(
   UVs=Panner(Coordinate=@n25) {speedX=0.001, speedY=0.001}
   Tex=TextureObject() {samplerType=SAMPLERTYPE_Grayscale, texture=/FortUILibrary/MaterialFunctions/ExampleContent/T_Cloud_R}
  )
  n29 = Constant() {r=0.5}
  n30 = Constant() {r=0.07}
  n31 = Constant() {r=0.25}
  NamedRerouteDeclaration(None=ComponentMask(None=MaterialFunctionCall!(UVs=@n26, Texture Object=@n27, Grayscale Flow Map=@n28.RGB, Flow Direction=@n29, Flow Intensity=@n30, Flow Speed=@n31) {materialFunction=/FortUILibrary/MaterialFunctions/Effects/MF_UI_GrayscaleFlowMap_New}.Distorted Texture) {r=true}) {name=FlowMap_1, nodeColor=<rgba 1 0.4922 0 1>}
  n32 = Multiply(A=@n4) {constB=0.7}
  n33 = Panner(Coordinate=@n32) {speedX=0.001, speedY=0.01}
  n34 = TextureObject() {samplerType=SAMPLERTYPE_Alpha, texture=&T_UI_CloudLV}
  n35 = TextureSample(
   UVs=Panner(Coordinate=@n32) {speedX=0.01, speedY=0.01}
   Tex=TextureObject() {samplerType=SAMPLERTYPE_Grayscale, texture=&T_Cloud_R}
  )
  n36 = Constant() {r=0.5}
  n37 = Constant() {r=0.01}
  n38 = Constant() {r=0.25}
  NamedRerouteDeclaration(None=ComponentMask(None=MaterialFunctionCall(UVs=@n33, Texture Object=@n34, Grayscale Flow Map=@n35.RGB, Flow Direction=@n36, Flow Intensity=@n37, Flow Speed=@n38) {materialFunction=&MF_UI_GrayscaleFlowMap_New}.Distorted Texture) {r=true}) {name=FlowMap_FG_1, nodeColor=<rgba 0 0 0 1>}
  NamedRerouteDeclaration(None=LinearInterpolate(
    A=Constant3Vector() {constant=<rgba 0.2712 0.0735 0.651 1>}
    B=Constant3Vector() {constant=<rgba 0.6194 0.1823 1 1>}
    Alpha=NamedRerouteUsage()
   )) {name=foreground clouds_1, nodeColor=<rgba 0 1 0.7188 1>}
  n39 = Multiply(A=AppendVector(A=Constant(), B=Constant() {r=0.85}), B=NamedRerouteUsage())
  n40 = Constant() {r=0.05}
  n41 = Constant() {r=2}
  n42 = Constant() {r=2}
  n43 = NamedRerouteUsage()
  n44 = Constant() {r=1.25}
  n45 = TextureSample(UVs=Panner(Coordinate=MaterialFunctionCall(UVs=@n43, Texture Scale=@n44) {materialFunction=&ScaleUVsByCenter}.UVs) {speedX=0.03, speedY=0.02}) {samplerType=SAMPLERTYPE_Grayscale, texture=&LowResBlurredNoise}
  n46 = Max(A=NamedRerouteUsage(), B=@n45.R)
  n47 = Multiply(
   A=OneMinus(None=MaterialFunctionCall(UVs=@n39, Subtract=@n40, Falloff=@n41, Intensity=@n42) {materialFunction=&MF_UI_TopBottomGradient}.Result)
   B=@n46
  )
  n48 = Max(A=@n47, B=@n46)
  n49 = Max(
   A=MaterialFunctionCall(GradientStart=Constant() {r=-0.1}, GradientEnd=Constant() {r=1.5}, RotationAngle=Constant() {r=0.8}, GradientOffset=Constant() {r=0.5}) {materialFunction=&MF_UI_LinearGradient}.Result
   B=MaterialFunctionCall(GradientStart=Constant() {r=-0.1}, GradientEnd=Constant() {r=1.25}, RotationAngle=Constant() {r=0.65}, GradientOffset=Constant() {r=0.55}) {materialFunction=&MF_UI_LinearGradient}.Result
  )
  n50 = Saturate(None=Multiply(A=Max(A=@n48, B=@n49), B=@n49))
  NamedRerouteDeclaration(None=SmoothStep(Value=Multiply(A=@n50) {constB=4}) {constMax=0.54}) {name=foreground cloud alpha_1, nodeColor=<rgba 0 1 0.8359 1>}
  NamedRerouteDeclaration(None=@n3) {name=sky gradient_1, nodeColor=<rgba 0 0.4609 1 1>}
  NamedRerouteDeclaration(None=Multiply(A=@n18) {constB=1.5}) {name=BG Cloud Alpha_1, nodeColor=<rgba 0.6641 1 0 1>}
  n51 = NamedRerouteUsage()
  n52 = Constant3Vector() {constant=<rgba 0.8551 0.183 1 1>}
  n53 = Constant3Vector() {constant=<rgba 0.7969 0.2075 0.792 1>}
  n54 = Constant3Vector() {constant=<rgba 0.2155 0.1629 0.8021 1>}
  n55 = NamedRerouteUsage()
  n56 = Constant() {r=0.15}
  n57 = Constant() {r=1}
  n58 = Constant() {r=0.75}
  n59 = Constant() {r=0.25}
  n60 = NamedRerouteUsage()
  n61 = Constant() {r=1.25}
  n62 = TextureSample(UVs=Panner(Coordinate=Multiply(A=@n60, B=@n61)) {speedX=0.01, speedY=0.01}) {samplerType=SAMPLERTYPE_Grayscale, texture=&LowResBlurredNoise}
  n63 = Multiply(A=@n62.R) {constB=1.1}
  n64 = Multiply(A=@n1, B=@n2)
  n65 = Constant() {r=-0.2}
  n66 = Constant() {r=0.9}
  n67 = Constant() {r=2}
  n68 = Multiply(
   A=OneMinus(None=MaterialFunctionCall(UVs=@n64, Subtract=@n65, Falloff=@n66, Intensity=@n67) {materialFunction=&MF_UI_TopBottomGradient}.Result)
   B=Max(A=NamedRerouteUsage(), B=@n62.R)
  )
  n69 = SmoothStep(Value=OneMinus(None=Add(A=@n63, B=@n68))) {constMax=0.5, constMin=0.1}
  n70 = Multiply(A=MaterialFunctionCall(GradientStart=Constant() {r=0.2}, GradientEnd=Constant() {r=0.65}, RotationAngle=Constant() {r=0.75}, GradientOffset=Constant() {r=0.2}) {materialFunction=&MF_UI_LinearGradient}.Result) {constB=0.2}
  n71 = Saturate(None=Multiply(A=@n69, B=Min(A=@n70, B=@n69)))
  n72 = LinearInterpolate(
   A=LinearInterpolate(A=NamedRerouteUsage(), B=@n24, Alpha=Multiply(A=@n51) {constB=1.2})
   B=LinearInterpolate(
    A=MaterialFunctionCall(A=@n52, B=@n53, C=@n54, Alpha=@n55) {materialFunction=&3ColorBlend}.Result
    B=@n54
    Alpha=MaterialFunctionCall(GradientStart=@n56, GradientEnd=@n57, RotationAngle=@n58, GradientOffset=@n59) {materialFunction=&MF_UI_LinearGradient}.Result
   )
   Alpha=Multiply(A=SmoothStep(Value=@n71)) {constB=1.25}
  )
  n73 = NamedRerouteUsage()
  n74 = NamedRerouteUsage()
  n75 = MaterialFunctionCall(GradientStart=Constant() {r=0.05}, GradientEnd=Constant() {r=0.25}, RotationAngle=Constant() {r=0.25}, GradientOffset=Constant() {r=-0.07}) {materialFunction=&MF_UI_LinearGradient}
  NamedRerouteDeclaration(None=LinearInterpolate(
    A=Constant3Vector() {constant=<rgba 0.0203 0.0103 0.147 1>}
    B=LinearInterpolate(A=@n72, B=@n73, Alpha=@n74)
    Alpha=SmoothStep(Value=@n75.Result)
   )) {name=BestBG, nodeColor=<rgba 1 0.7969 0 1>}
  NamedRerouteUsage()
  OneMinus(None=Add(A=@n45.R, B=@n47))
 ]
}
