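# Item Shop backdrop material for the LastVoice theme, drawing an animated purple and pink cloudy sky.
# Screen-space scaled UVs drive panning cloud textures distorted by grayscale flow maps for background and foreground layers.
# Blurred noise, linear and top-bottom gradients shape cloud masks that blend color ramps into a composite background.
# A partly desaturated color feeds the shared MPItemShop setup function, which also has floor inputs.
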Material /ItemShop/UI/Art/LastVoice/M_UI_LastVoice_DN_MP
graph {
 properties {pixelDepthOffsetMode=PDOM_Legacy, usedShadingModels=Default Lit}
 nodes [
  Constant3Vector() {constant=<rgba 0.5457 0.4406 0.5349 1>}
  Constant3Vector() {constant=<rgba 0.2024 0.1008 0.6719 1>}
  Constant3Vector() {constant=<rgba 0.5397 0.2349 0.9047 1>}
  MaterialFunctionCall!(
   Mask
   CloudIntensity=Constant() {r=1}
   Density=Constant() {r=1}
   Scale=Constant() {r=1}
   Speed
   CloudColorLow
   CloudColorHigh=Constant3Vector() {constant=<rgba 0.8963 0.8796 0.9047 1>}
   CloudColorDark=Constant3Vector() {constant=<rgba 0.5457 0.1594 0.3718 1>}
   CloudDarkIntensity=Constant() {r=2}
   FoggyAlphaMin=Constant() {r=5}
   NoiseTexture=TextureObject() {samplerType=SAMPLERTYPE_Alpha, texture=/Game/UI/Frontend/ItemShop/Art/LastVoice/T_UI_CloudLV}
  ) {materialFunction=/FortUILibrary/MaterialFunctions/Effects/MF_UI_Clouds}
  n1 = NamedRerouteUsage()
  MaterialFunctionCall(UVs=@n1, Texture Scale=Constant() {r=0.8}) {materialFunction=&ScaleUVsByCenter}
  MaterialFunctionCall() {materialFunction=&MF_UI_UVScaling}
  MaterialFunctionCall!(
   Color=Desaturation(None=NamedRerouteUsage(), Fraction=Constant() {r=0.5})
   Color Brightness=Constant() {r=1.2}
   Floor Radius
   Floor Density
   Floor Metallic
   Floor Specular
   Floor Roughness
  ) {materialFunction=/Game/UI/Foundation/Environment/Shell/Materials/MaterialFunctions/MF_UI_MPItemShop_Setup}
  n2 = Constant3Vector() {constant=<rgba 0.0203 0.0103 0.147 1>}
  MaterialFunctionCall(
   A=@n2
   B=Constant3Vector() {constant=<rgba 0.0382 0.2961 0.6939 1>}
   C=Constant3Vector() {constant=<rgba 0.8178 0.8459 0.9047 1>}
   Alpha=MaterialFunctionCall(GradientStart=Constant() {r=-0.1}, GradientEnd=Constant() {r=0.45}, RotationAngle=Constant() {r=0.72}, GradientOffset=Constant() {r=0.1}) {materialFunction=&MF_UI_LinearGradient}.Result
  ) {materialFunction=&3ColorBlend}
  n3 = NamedRerouteUsage()
  n4 = Constant() {r=0.75}
  n5 = TextureSample(UVs=Panner(Coordinate=Multiply(A=@n3, B=@n4)) {speedX=0.01, speedY=0.01}) {samplerType=SAMPLERTYPE_Grayscale, texture=&LowResBlurredNoise}
  n6 = Multiply(A=@n5.R) {constB=1.5}
  n7 = Constant()
  n8 = Constant() {r=4}
  n9 = MaterialFunctionCall!(
   UVs=Multiply(A=AppendVector(A=@n7, B=@n8), B=NamedRerouteUsage())
   Subtract=Constant()
   Falloff=Constant() {r=0.1}
   Intensity=Constant() {r=2}
  ) {materialFunction=/Game/UI/Frontend/ItemShop/Art/_MaterialFunctions_ShopTheming/MF_UI_TopBottomGradient}
  n10 = NamedRerouteUsage()
  n11 = Multiply(A=OneMinus(None=SmoothStep(Value=@n9.Result)), B=Max(A=Multiply(A=@n10) {constB=1.25}, B=@n5.R))
  n12 = SmoothStep(Value=OneMinus(None=Add(A=@n6, B=@n11))) {constMax=0.35}
  n13 = Constant()
  n14 = Constant() {r=0.1}
  n15 = MaterialFunctionCall(
   UVs=Add(A=NamedRerouteUsage(), B=AppendVector(A=@n13, B=@n14))
   GradientStart=Constant() {r=-0.01}
   GradientEnd=Constant() {r=0.96}
   RotationAngle=Constant() {r=0.75}
   GradientOffset=Constant() {r=0.1}
  ) {materialFunction=&MF_UI_LinearGradient}
  n16 = Saturate(None=Multiply(A=@n12, B=Min(A=@n15.Result, B=@n12)))
  n17 = Constant3Vector() {constant=<rgba 0.107 0.0319 0.4125 1>}
  n18 = NamedRerouteUsage()
  n19 = Constant() {r=0.15}
  n20 = Constant() {r=1}
  n21 = Constant() {r=0.75}
  n22 = Constant() {r=0.25}
  n23 = LinearInterpolate(
   A=MaterialFunctionCall(
    A=Constant3Vector() {constant=<rgba 0.8423 0.3128 0.9688 1>}
    B=Constant3Vector() {constant=<rgba 0.3046 0.0744 0.7031 1>}
    C=@n17
    Alpha=NamedRerouteUsage()
   ) {materialFunction=&3ColorBlend}.Result
   B=@n17
   Alpha=SmoothStep(Value=MaterialFunctionCall(UVs=@n18, GradientStart=@n19, GradientEnd=@n20, RotationAngle=@n21, GradientOffset=@n22) {materialFunction=&MF_UI_LinearGradient}.Result)
  )
  Multiply(A=@n16, B=@n23)
  NamedRerouteDeclaration(None=MaterialFunctionCall!(ScreenUV, TextureSize, ScalePivotPoint) {materialFunction=/FortUILibrary/MaterialFunctions/Basics/MF_UI_ScreenSpaceUVScaling}.Scaled to Fit) {name=UVScaling_1, nodeColor=<rgba 0 0.9297 1 1>}
  n24 = Multiply(A=NamedRerouteUsage()) {constB=1.05}
  n25 = Panner(Coordinate=@n24) {speedX=0.01, speedY=0.001}
  n26 = TextureObject() {samplerType=SAMPLERTYPE_Alpha, texture=&T_UI_CloudLV}
  n27 = TextureSample(
   UVs=Panner(Coordinate=@n24) {speedX=0.001, speedY=0.001}
   Tex=TextureObject() {samplerType=SAMPLERTYPE_Grayscale, texture=/FortUILibrary/MaterialFunctions/ExampleContent/T_Cloud_R}
  )
  n28 = Constant() {r=0.5}
  n29 = Constant() {r=0.07}
  n30 = Constant() {r=0.25}
  NamedRerouteDeclaration(None=ComponentMask(None=MaterialFunctionCall!(UVs=@n25, Texture Object=@n26, Grayscale Flow Map=@n27.RGB, Flow Direction=@n28, Flow Intensity=@n29, Flow Speed=@n30) {materialFunction=/FortUILibrary/MaterialFunctions/Effects/MF_UI_GrayscaleFlowMap_New}.Distorted Texture) {r=true}) {name=FlowMap_1, nodeColor=<rgba 1 0.4922 0 1>}
  n31 = Multiply(A=@n1) {constB=0.7}
  n32 = Panner(Coordinate=@n31) {speedX=0.001, speedY=0.01}
  n33 = TextureObject() {samplerType=SAMPLERTYPE_Alpha, texture=&T_UI_CloudLV}
  n34 = TextureSample(
   UVs=Panner(Coordinate=@n31) {speedX=0.01, speedY=0.01}
   Tex=TextureObject() {samplerType=SAMPLERTYPE_Grayscale, texture=&T_Cloud_R}
  )
  n35 = Constant() {r=0.5}
  n36 = Constant() {r=0.01}
  n37 = Constant() {r=0.25}
  NamedRerouteDeclaration(None=ComponentMask(None=MaterialFunctionCall(UVs=@n32, Texture Object=@n33, Grayscale Flow Map=@n34.RGB, Flow Direction=@n35, Flow Intensity=@n36, Flow Speed=@n37) {materialFunction=&MF_UI_GrayscaleFlowMap_New}.Distorted Texture) {r=true}) {name=FlowMap_FG_1, nodeColor=<rgba 0 0 0 1>}
  n38 = Constant3Vector() {constant=<rgba 0.0203 0.0103 0.147 1>}
  n39 = NamedRerouteUsage()
  n40 = Multiply(A=NamedRerouteUsage()) {constB=1.2}
  n41 = Constant3Vector() {constant=<rgba 0.2155 0.1629 0.8021 1>}
  n42 = MaterialFunctionCall(
   A=Constant3Vector() {constant=<rgba 0.8551 0.183 1 1>}
   B=Constant3Vector() {constant=<rgba 0.7969 0.2075 0.792 1>}
   C=@n41
   Alpha=NamedRerouteUsage()
  ) {materialFunction=&3ColorBlend}
  n43 = MaterialFunctionCall(
   UVs=NamedRerouteUsage()
   GradientStart=Constant() {r=0.15}
   GradientEnd=Constant() {r=1}
   RotationAngle=Constant() {r=0.75}
   GradientOffset=Constant() {r=0.25}
  ) {materialFunction=&MF_UI_LinearGradient}
  n44 = NamedRerouteUsage()
  n45 = Constant() {r=1.25}
  n46 = TextureSample(UVs=Panner(Coordinate=Multiply(A=@n44, B=@n45)) {speedX=0.01, speedY=0.01}) {samplerType=SAMPLERTYPE_Grayscale, texture=&LowResBlurredNoise}
  n47 = Multiply(A=@n46.R) {constB=1.1}
  n48 = Multiply(A=AppendVector(A=Constant(), B=Constant() {r=2}))
  n49 = Constant() {r=-0.2}
  n50 = Constant() {r=0.9}
  n51 = Constant() {r=2}
  n52 = Multiply(
   A=OneMinus(None=MaterialFunctionCall(UVs=@n48, Subtract=@n49, Falloff=@n50, Intensity=@n51) {materialFunction=&MF_UI_TopBottomGradient}.Result)
   B=Max(A=NamedRerouteUsage(), B=@n46.R)
  )
  n53 = SmoothStep(Value=OneMinus(None=Add(A=@n47, B=@n52))) {constMax=0.5, constMin=0.1}
  n54 = NamedRerouteUsage()
  n55 = Constant() {r=0.2}
  n56 = Constant() {r=0.65}
  n57 = Constant() {r=0.75}
  n58 = Constant() {r=0.2}
  n59 = Min(
   A=Multiply(A=MaterialFunctionCall(UVs=@n54, GradientStart=@n55, GradientEnd=@n56, RotationAngle=@n57, GradientOffset=@n58) {materialFunction=&MF_UI_LinearGradient}.Result) {constB=0.2}
   B=@n53
  )
  n60 = SmoothStep(Value=Saturate(None=Multiply(A=@n53, B=@n59)))
  n61 = LinearInterpolate(
   A=LinearInterpolate(
    A=LinearInterpolate(A=@n39, B=@n23, Alpha=@n40)
    B=LinearInterpolate(A=@n42.Result, B=@n41, Alpha=@n43.Result)
    Alpha=Multiply(A=@n60) {constB=1.25}
   )
   B=NamedRerouteUsage()
   Alpha=NamedRerouteUsage()
  )
  n62 = SmoothStep(Value=MaterialFunctionCall(
    UVs=NamedRerouteUsage()
    GradientStart=Constant() {r=0.05}
    GradientEnd=Constant() {r=0.8}
    RotationAngle=Constant() {r=0.29}
    GradientOffset=Constant() {r=-0.1}
   ) {materialFunction=&MF_UI_LinearGradient}.Result) {constMax=1.25}
  NamedRerouteDeclaration(None=LinearInterpolate(A=LinearInterpolate(A=@n38, B=@n61, Alpha=@n62), B=Constant3Vector() {constant=<rgba 0.0392 0.0141 0.1354 1>}) {constAlpha=0.3}) {name=BestBG, nodeColor=<rgba 1 0.7969 0 1>}
  NamedRerouteDeclaration(None=LinearInterpolate(
    A=Constant3Vector() {constant=<rgba 0.2712 0.0735 0.651 1>}
    B=Constant3Vector() {constant=<rgba 0.7802 0.2192 1 1>}
    Alpha=NamedRerouteUsage()
   )) {name=foreground clouds_1, nodeColor=<rgba 0 1 0.7188 1>}
  n63 = Multiply(A=AppendVector(A=Constant(), B=Constant() {r=0.85}), B=NamedRerouteUsage())
  n64 = Constant() {r=0.05}
  n65 = Constant() {r=2}
  n66 = Constant() {r=2}
  n67 = NamedRerouteUsage()
  n68 = Constant() {r=1.25}
  n69 = TextureSample(UVs=Panner(Coordinate=MaterialFunctionCall(UVs=@n67, Texture Scale=@n68) {materialFunction=&ScaleUVsByCenter}.UVs) {speedX=0.03, speedY=0.02}) {samplerType=SAMPLERTYPE_Grayscale, texture=&LowResBlurredNoise}
  n70 = Max(A=NamedRerouteUsage(), B=@n69.R)
  n71 = Multiply(
   A=OneMinus(None=MaterialFunctionCall(UVs=@n63, Subtract=@n64, Falloff=@n65, Intensity=@n66) {materialFunction=&MF_UI_TopBottomGradient}.Result)
   B=@n70
  )
  n72 = Max(A=@n71, B=@n70)
  n73 = Max(
   A=MaterialFunctionCall(
    UVs=NamedRerouteUsage()
    GradientStart=Constant() {r=-0.1}
    GradientEnd=Constant() {r=1.5}
    RotationAngle=Constant() {r=0.8}
    GradientOffset=Constant() {r=0.5}
   ) {materialFunction=&MF_UI_LinearGradient}.Result
   B=MaterialFunctionCall(
    UVs=NamedRerouteUsage()
    GradientStart=Constant() {r=-0.1}
    GradientEnd=Constant() {r=1.25}
    RotationAngle=Constant() {r=0.65}
    GradientOffset=Constant() {r=0.55}
   ) {materialFunction=&MF_UI_LinearGradient}.Result
  )
  n74 = Saturate(None=Multiply(A=Max(A=@n72, B=@n73), B=@n73))
  NamedRerouteDeclaration(None=SmoothStep(Value=Multiply(A=@n74) {constB=4}) {constMax=0.54}) {name=foreground cloud alpha_1, nodeColor=<rgba 0 1 0.8359 1>}
  NamedRerouteDeclaration(None=@n2) {name=sky gradient_1, nodeColor=<rgba 0 0.4609 1 1>}
  NamedRerouteDeclaration(None=Multiply(A=@n16) {constB=1.5}) {name=BG Cloud Alpha_1, nodeColor=<rgba 0.6641 1 0 1>}
  OneMinus(None=Add(A=@n69.R, B=@n71))
 ]
}
