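# UI material function that animates a texture with flow-map distortion, returning the blended RGBA.
# A 0–1 direction input becomes a vector through saturated ramps, blended from a neutral value by flow-map red/green times intensity.
# That vector offsets the UVs of two samples cycling half a phase apart, crossfaded by a triangle wave to hide resets.
# The flow map's red channel also offsets the time phase per pixel, with speed scaling time.
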MaterialFunction /FortUILibrary/MaterialFunctions/Effects/MF_UI_GrayscaleFlowMap
graph {
 properties {description=Single chanel or 2 channel flow map directional material, previewMaterialDomain=MD_UI, libraryCategoriesText [UI Shader Library]}
 nodes [
  ComponentMask(None=TextureCoordinate()) {r=true}
  n1 = FunctionInput(Preview=TextureCoordinate()) {
   bUsePreviewValueAsDefault=true
   description=The Teture UVs to recieve the flow map distortion
   inputName=Texture UVs
   inputType=FunctionInput_Vector2
   sortPriority=1
  }
  n2 = FunctionInput(Preview=TextureCoordinate()) {
   bUsePreviewValueAsDefault=true
   description=The uvs of the flow map texture
   inputName=FlowMap Uvs
   inputType=FunctionInput_Vector2
   sortPriority=3
  }
  n3 = FunctionInput(Preview=TextureObject()) {
   description=The flow map texture that can be 2 channel or single channel
   inputName=FlowMap Texture T2D
   inputType=FunctionInput_Texture2D
   sortPriority=2
  }
  n4 = TextureSample(UVs=@n2, Tex=@n3)
  n5 = FunctionInput(Preview=Constant() {r=0.25}) {bUsePreviewValueAsDefault=true, description=The speed of the flow map, inputName=FlowSpeed, inputType=FunctionInput_Scalar, sortPriority=6}
  n6 = Add(A=Max(A=@n4.R, B=@n4.R) {desc=I like max here dunno why - Adrienne}, B=Multiply(A=Time(), B=@n5) {constB=4})
  n7 = Frac(None=@n6)
  n8 = Add(A=@n1, B=Multiply(A=NamedRerouteUsage(), B=@n7) {constB=4})
  n9 = FunctionInput(Preview=TextureObject()) {description=The texture the flow map applies to, inputName=Texture 2D, inputType=FunctionInput_Texture2D}
  n10 = Add(A=@n6) {constB=0.5}
  n11 = Add(A=@n1, B=Multiply(A=NamedRerouteUsage(), B=Frac(None=@n10)) {constB=4})
  n12 = ConstantBiasScale(None=@n7) {bias=-0.5, scale=2}
  FunctionOutput(None=LinearInterpolate(A=TextureSample(UVs=@n8, Tex=@n9).RGBA, B=TextureSample(UVs=@n11, Tex=@n9).RGBA, Alpha=Abs(None=@n12)))
  n13 = Constant2Vector() {desc=Not Moving Flow Map, g=0.5, r=0.5}
  n14 = FunctionInput(Preview=Constant() {r=0.25}) {
   bUsePreviewValueAsDefault=true
   description=This is a 0-1 determineing the flow map direction
   inputName=FlowDirection
   inputType=FunctionInput_Scalar
   sortPriority=4
  }
  n15 = Frac(None=@n14)
  n16 = Multiply(A=Subtract(A=@n15) {constB=0.25}) {constB=4}
  n17 = OneMinus(None=@n15)
  n18 = Multiply(A=@n17) {constB=4}
  n19 = Multiply(A=Subtract(A=@n17) {constB=0.25}) {constB=4}
  n20 = Multiply(A=@n15) {constB=4}
  n21 = AppendVector(
   A=Multiply(A=Saturate(None=@n16), B=Saturate(None=@n18)) {constB=4, desc=XCoords}
   B=Multiply(A=Saturate(None=@n19), B=Saturate(None=@n20)) {constB=4, desc=YCoords}
  )
  n22 = Multiply(
   A=ComponentMask(None=@n4.RGB) {g=true, r=true}
   B=FunctionInput(Preview=Constant() {r=0.4}) {bUsePreviewValueAsDefault=true, inputName=FlowIntensity, inputType=FunctionInput_Scalar, sortPriority=5}
  ) {constB=4}
  NamedRerouteDeclaration(None=ConstantBiasScale(None=LinearInterpolate(A=@n13, B=@n21, Alpha=@n22)) {bias=-0.5, scale=2}) {name=FlowMapCoordinates, nodeColor=<rgba 0 0 0 1>}
 ]
}
