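# UI material that draws an animated, stylized cloud sky for the LastVoice item shop background.
# Two panning flow-mapped cloud textures and blurred noise layers, shaped by linear and top-bottom gradients, form background and foreground cloud masks.
# These masks are set up alongside gradient-blended pink, lavender and blue colors, with nested lerps layering clouds and sky and a linear gradient fading to purple.
# Named reroutes organize the shared UV scaling, masks and colors.
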Material /ItemShop/UI/Art/LastVoice/M_UI_LastVoice_ShopBG
graph {
 properties {bUsedWithStaticMesh=false, materialDomain=MD_UI, pixelDepthOffsetMode=PDOM_Legacy, usedShadingModels=Default Lit}
 nodes [
  n1 = LinearInterpolate(A=NamedRerouteUsage(), B=NamedRerouteUsage(), Alpha=Multiply(A=NamedRerouteUsage()) {constB=1.2})
  LinearInterpolate(A=@n1, B=@n1, Alpha=NamedRerouteUsage())
  LinearInterpolate(
   A=Constant3Vector() {constant=<rgba 0.1894 0.0793 0.4844 1>}
   B=LinearInterpolate(A=@n1, B=NamedRerouteUsage(), Alpha=NamedRerouteUsage())
   Alpha=MaterialFunctionCall(GradientStart=Constant() {r=0.05}, GradientEnd=Constant() {r=0.25}, RotationAngle=Constant() {r=0.25}, GradientOffset=Constant() {r=-0.07}) {materialFunction=&MF_UI_LinearGradient}.Result
  )
  n2 = NamedRerouteUsage()
  MaterialFunctionCall(UVs=@n2, Texture Scale=Constant() {r=0.8}) {materialFunction=&ScaleUVsByCenter}
  n3 = TextureObject() {samplerType=SAMPLERTYPE_Alpha, texture=/Game/UI/Frontend/ItemShop/Art/LastVoice/T_UI_CloudLV}
  n4 = Multiply(A=@n2) {constB=0.65}
  n5 = Panner(Coordinate=@n4) {speedX=0.001, speedY=0.01}
  n6 = TextureObject() {samplerType=SAMPLERTYPE_Grayscale, texture=/FortUILibrary/MaterialFunctions/ExampleContent/T_Cloud_R}
  n7 = Panner(Coordinate=@n4) {speedX=0.01, speedY=0.01}
  n8 = Constant() {r=0.5}
  n9 = Constant() {r=0.01}
  n10 = Constant() {r=0.25}
  NamedRerouteDeclaration(None=ComponentMask(None=MaterialFunctionCall!(Texture 2D=@n3, Texture UVs=@n5, FlowMap Texture T2D=@n6, FlowMap Uvs=@n7, FlowDirection=@n8, FlowIntensity=@n9, FlowSpeed=@n10) {materialFunction=/FortUILibrary/MaterialFunctions/Effects/MF_UI_GrayscaleFlowMap}.Result) {r=true}) {name=FlowMap_FG_2, nodeColor=<rgba 0 0 0 1>}
  NamedRerouteDeclaration(None=LinearInterpolate(
    A=Constant3Vector() {constant=<rgba 0.8469 0.4543 0.6425 1>}
    B=Constant3Vector() {constant=<rgba 0.9301 0.8632 0.9131 1>}
    Alpha=NamedRerouteUsage()
   )) {name=foreground clouds_2, nodeColor=<rgba 0 1 0.7188 1>}
  n11 = Multiply(A=AppendVector(A=Constant(), B=Constant() {r=0.85}), B=NamedRerouteUsage())
  n12 = Constant() {r=0.05}
  n13 = Constant() {r=1}
  n14 = Constant() {r=2}
  n15 = NamedRerouteUsage()
  n16 = Constant() {r=1.25}
  n17 = TextureSample(UVs=Panner(Coordinate=MaterialFunctionCall(UVs=@n15, Texture Scale=@n16) {materialFunction=&ScaleUVsByCenter}.UVs) {speedX=0.03, speedY=0.02}) {samplerType=SAMPLERTYPE_Grayscale, texture=&LowResBlurredNoise}
  n18 = Max(A=NamedRerouteUsage(), B=@n17.R)
  n19 = Multiply(
   A=OneMinus(None=MaterialFunctionCall!(UVs=@n11, Subtract=@n12, Falloff=@n13, Intensity=@n14) {materialFunction=/Game/UI/Frontend/ItemShop/Art/_MaterialFunctions_ShopTheming/MF_UI_TopBottomGradient}.Result)
   B=@n18
  )
  n20 = Max(A=@n19, B=@n18)
  n21 = Max(
   A=MaterialFunctionCall(GradientStart=Constant() {r=-0.1}, GradientEnd=Constant() {r=1.25}, RotationAngle=Constant() {r=0.8}, GradientOffset=Constant() {r=0.5}) {materialFunction=&MF_UI_LinearGradient}.Result
   B=MaterialFunctionCall(GradientStart=Constant() {r=-0.1}, GradientEnd=Constant() {r=1.25}, RotationAngle=Constant() {r=0.65}, GradientOffset=Constant() {r=0.55}) {materialFunction=&MF_UI_LinearGradient}.Result
  )
  n22 = Saturate(None=Multiply(A=Max(A=@n20, B=@n21), B=@n21))
  NamedRerouteDeclaration(None=SmoothStep(Value=Multiply(A=@n22) {constB=4}) {constMax=0.54}) {name=foreground cloud alpha_2, nodeColor=<rgba 0 1 0.8359 1>}
  NamedRerouteDeclaration(None=Constant3Vector() {constant=<rgba 0.0482 0.3419 0.7991 1>}) {name=sky gradient_2, nodeColor=<rgba 0 0.4609 1 1>}
  n23 = NamedRerouteUsage()
  n24 = Constant() {r=0.75}
  n25 = TextureSample(UVs=Panner(Coordinate=Multiply(A=@n23, B=@n24)) {speedX=0.01, speedY=0.01}) {samplerType=SAMPLERTYPE_Grayscale, texture=&LowResBlurredNoise}
  n26 = Multiply(A=@n25.R) {constB=1.5}
  n27 = Multiply(A=AppendVector(A=Constant(), B=Constant() {r=4.5}), B=NamedRerouteUsage())
  n28 = Constant()
  n29 = Constant() {r=0.2}
  n30 = Constant() {r=2}
  n31 = NamedRerouteUsage()
  n32 = Multiply(
   A=OneMinus(None=MaterialFunctionCall(UVs=@n27, Subtract=@n28, Falloff=@n29, Intensity=@n30) {materialFunction=&MF_UI_TopBottomGradient}.Result)
   B=Max(A=Multiply(A=@n31) {constB=1.25}, B=@n25.R)
  )
  n33 = SmoothStep(Value=OneMinus(None=Add(A=@n26, B=@n32))) {constMax=0.35}
  n34 = Constant() {r=-0.01}
  n35 = Constant() {r=0.6}
  n36 = Constant() {r=0.75}
  n37 = Constant() {r=0.1}
  n38 = Multiply(
   A=@n33
   B=Min(
    A=MaterialFunctionCall(GradientStart=@n34, GradientEnd=@n35, RotationAngle=@n36, GradientOffset=@n37) {materialFunction=&MF_UI_LinearGradient}.Result
    B=@n33
   )
  )
  NamedRerouteDeclaration(None=Multiply(A=Saturate(None=@n38)) {constB=1.5}) {name=BG Cloud Alpha_2, nodeColor=<rgba 0.6641 1 0 1>}
  n39 = Constant3Vector() {constant=<rgba 0.9911 0.893 0.9558 1>}
  n40 = Constant3Vector() {constant=<rgba 0.4007 0.375 0.9115 1>}
  n41 = Constant3Vector() {constant=<rgba 0.1279 0.268 0.5833 1>}
  n42 = NamedRerouteUsage()
  n43 = Constant() {r=0.15}
  n44 = Constant() {r=1}
  n45 = Constant() {r=0.75}
  n46 = Constant() {r=0.25}
  NamedRerouteDeclaration(None=LinearInterpolate(
    A=MaterialFunctionCall(A=@n39, B=@n40, C=@n41, Alpha=@n42) {materialFunction=&3ColorBlend}.Result
    B=@n41
    Alpha=MaterialFunctionCall(GradientStart=@n43, GradientEnd=@n44, RotationAngle=@n45, GradientOffset=@n46) {materialFunction=&MF_UI_LinearGradient}.Result
   )) {name=Shadow Colors, nodeColor=<rgba 1 0.0938 0 1>}
  n47 = Constant3Vector() {constant=<rgba 0.7801 0.5833 1 1>}
  n48 = Constant3Vector() {constant=<rgba 0.7969 0.4441 0.6856 1>}
  n49 = Constant3Vector() {constant=<rgba 0.2772 0.2302 0.8021 1>}
  n50 = NamedRerouteUsage()
  n51 = Constant() {r=0.15}
  n52 = Constant() {r=1}
  n53 = Constant() {r=0.75}
  n54 = Constant() {r=0.25}
  NamedRerouteDeclaration(None=LinearInterpolate(
    A=MaterialFunctionCall(A=@n47, B=@n48, C=@n49, Alpha=@n50) {materialFunction=&3ColorBlend}.Result
    B=@n49
    Alpha=MaterialFunctionCall(GradientStart=@n51, GradientEnd=@n52, RotationAngle=@n53, GradientOffset=@n54) {materialFunction=&MF_UI_LinearGradient}.Result
   )) {name=BG Cloud Color, nodeColor=<rgba 0 0.2266 1 1>}
  n55 = NamedRerouteUsage()
  n56 = Constant() {r=1.25}
  n57 = TextureSample(UVs=Panner(Coordinate=Multiply(A=@n55, B=@n56)) {speedX=0.01, speedY=0.01}) {samplerType=SAMPLERTYPE_Grayscale, texture=&LowResBlurredNoise}
  n58 = Multiply(A=@n57.R) {constB=1.1}
  n59 = Multiply(A=AppendVector(A=Constant(), B=Constant() {r=5}), B=NamedRerouteUsage())
  n60 = Constant()
  n61 = Constant() {r=0.2}
  n62 = Constant() {r=2}
  n63 = Multiply(
   A=OneMinus(None=MaterialFunctionCall(UVs=@n59, Subtract=@n60, Falloff=@n61, Intensity=@n62) {materialFunction=&MF_UI_TopBottomGradient}.Result)
   B=Max(A=NamedRerouteUsage(), B=@n57.R)
  )
  n64 = SmoothStep(Value=OneMinus(None=Add(A=@n58, B=@n63))) {constMax=0.5, constMin=0.1}
  n65 = Constant() {r=0.2}
  n66 = Constant() {r=0.65}
  n67 = Constant() {r=0.75}
  n68 = Constant() {r=0.2}
  n69 = Multiply(
   A=@n64
   B=Min(
    A=MaterialFunctionCall(GradientStart=@n65, GradientEnd=@n66, RotationAngle=@n67, GradientOffset=@n68) {materialFunction=&MF_UI_LinearGradient}.Result
    B=@n64
   )
  )
  NamedRerouteDeclaration(None=Multiply(A=Saturate(None=@n69)) {constB=1.25}) {name=BG Cloud Mask, nodeColor=<rgba 0.5703 1 0 1>}
  NamedRerouteDeclaration(None=MaterialFunctionCall() {materialFunction=&MF_UI_UVScaling}.Scaled to Fit) {name=UVScaling_2, nodeColor=<rgba 0 0.9297 1 1>}
  n70 = TextureObject() {samplerType=SAMPLERTYPE_Alpha, texture=&T_UI_CloudLV}
  n71 = Multiply(A=NamedRerouteUsage()) {constB=1.05}
  n72 = Panner(Coordinate=@n71) {speedX=0.01, speedY=0.001}
  n73 = TextureObject() {samplerType=SAMPLERTYPE_Grayscale, texture=&T_Cloud_R}
  n74 = Panner(Coordinate=@n71) {speedX=0.001, speedY=0.001}
  n75 = Constant() {r=0.5}
  n76 = Constant() {r=0.07}
  n77 = Constant() {r=0.25}
  NamedRerouteDeclaration(None=ComponentMask(None=MaterialFunctionCall(Texture 2D=@n70, Texture UVs=@n72, FlowMap Texture T2D=@n73, FlowMap Uvs=@n74, FlowDirection=@n75, FlowIntensity=@n76, FlowSpeed=@n77) {materialFunction=&MF_UI_GrayscaleFlowMap}.Result) {r=true}) {name=FlowMap_2, nodeColor=<rgba 1 0.4922 0 1>}
  OneMinus(None=Add(A=@n17.R, B=@n19))
 ]
}
