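# Animated Last Voice themed Item Shop background material: a pastel pink and purple cloudscape over blue sky, fed to the MPItemShop setup function.
# Two panning cloud textures, distorted by grayscale flow maps on scaled screen-space UVs, supply cloud detail.
# Gradients and scrolling blurred noise shape masks for layered background, middle and foreground clouds.
# Lerps stack the tinted layers, with a final linear gradient blending in deep purple.
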Material /ItemShop/UI/Art/LastVoice/M_UI_LastVoice_ShopBG_MP
graph {
 properties {pixelDepthOffsetMode=PDOM_Legacy, usedShadingModels=Default Lit}
 nodes [
  MaterialFunctionCall!(Color=NamedRerouteUsage(), Color Brightness=Constant() {r=0.5}, Floor Radius, Floor Density, Floor Metallic, Floor Specular, Floor Roughness) {materialFunction=/Game/UI/Foundation/Environment/Shell/Materials/MaterialFunctions/MF_UI_MPItemShop_Setup}
  NamedRerouteDeclaration(None=MaterialFunctionCall!(ScreenUV, TextureSize, ScalePivotPoint) {materialFunction=/FortUILibrary/MaterialFunctions/Basics/MF_UI_ScreenSpaceUVScaling}.Scaled to Fit) {name=UVScaling_1, nodeColor=<rgba 0 0.9297 1 1>}
  n1 = Multiply(A=NamedRerouteUsage()) {constB=1.05}
  n2 = Panner(Coordinate=@n1) {speedX=0.01, speedY=0.001}
  n3 = TextureObject() {samplerType=SAMPLERTYPE_Alpha, texture=/Game/UI/Frontend/ItemShop/Art/LastVoice/T_UI_CloudLV}
  n4 = TextureSample(
   UVs=Panner(Coordinate=@n1) {speedX=0.001, speedY=0.001}
   Tex=TextureObject() {samplerType=SAMPLERTYPE_Grayscale, texture=/FortUILibrary/MaterialFunctions/ExampleContent/T_Cloud_R}
  )
  n5 = Constant() {r=0.5}
  n6 = Constant() {r=0.07}
  n7 = Constant() {r=0.25}
  NamedRerouteDeclaration(None=ComponentMask(None=MaterialFunctionCall!(UVs=@n2, Texture Object=@n3, Grayscale Flow Map=@n4.RGB, Flow Direction=@n5, Flow Intensity=@n6, Flow Speed=@n7) {materialFunction=/FortUILibrary/MaterialFunctions/Effects/MF_UI_GrayscaleFlowMap_New}.Distorted Texture) {r=true}) {name=FlowMap_1, nodeColor=<rgba 1 0.4922 0 1>}
  n8 = Multiply(A=NamedRerouteUsage()) {constB=0.65}
  n9 = Panner(Coordinate=@n8) {speedX=0.001, speedY=0.01}
  n10 = TextureObject() {samplerType=SAMPLERTYPE_Alpha, texture=&T_UI_CloudLV}
  n11 = TextureSample(
   UVs=Panner(Coordinate=@n8) {speedX=0.01, speedY=0.01}
   Tex=TextureObject() {samplerType=SAMPLERTYPE_Grayscale, texture=&T_Cloud_R}
  )
  n12 = Constant() {r=0.5}
  n13 = Constant() {r=0.01}
  n14 = Constant() {r=0.25}
  NamedRerouteDeclaration(None=ComponentMask(None=MaterialFunctionCall(UVs=@n9, Texture Object=@n10, Grayscale Flow Map=@n11.RGB, Flow Direction=@n12, Flow Intensity=@n13, Flow Speed=@n14) {materialFunction=&MF_UI_GrayscaleFlowMap_New}.Distorted Texture) {r=true}) {name=FlowMap_FG_1, nodeColor=<rgba 0 0 0 1>}
  NamedRerouteDeclaration(None=LinearInterpolate(
    A=Constant3Vector() {constant=<rgba 0.8469 0.4543 0.6425 1>}
    B=Constant3Vector() {constant=<rgba 0.9301 0.8632 0.9131 1>}
    Alpha=NamedRerouteUsage()
   )) {name=foreground clouds_1, nodeColor=<rgba 0 1 0.7188 1>}
  n15 = Constant()
  n16 = Constant() {r=0.85}
  n17 = MaterialFunctionCall!(
   UVs=Multiply(A=AppendVector(A=@n15, B=@n16), B=NamedRerouteUsage())
   Subtract=Constant() {r=0.05}
   Falloff=Constant() {r=1}
   Intensity=Constant() {r=2}
  ) {materialFunction=/Game/UI/Frontend/ItemShop/Art/_MaterialFunctions_ShopTheming/MF_UI_TopBottomGradient}
  n18 = MaterialFunctionCall(UVs=NamedRerouteUsage(), Texture Scale=Constant() {r=1.25}) {materialFunction=&ScaleUVsByCenter}
  n19 = Max(
   A=NamedRerouteUsage()
   B=TextureSample(UVs=Panner(Coordinate=@n18.UVs) {speedX=0.03, speedY=0.02}) {samplerType=SAMPLERTYPE_Grayscale, texture=&LowResBlurredNoise}.R
  )
  n20 = Max(A=Multiply(A=OneMinus(None=@n17.Result), B=@n19), B=@n19)
  n21 = Max(
   A=MaterialFunctionCall(
    UVs=NamedRerouteUsage()
    GradientStart=Constant() {r=-0.1}
    GradientEnd=Constant() {r=1.25}
    RotationAngle=Constant() {r=0.8}
    GradientOffset=Constant() {r=0.5}
   ) {materialFunction=&MF_UI_LinearGradient}.Result
   B=MaterialFunctionCall(
    UVs=NamedRerouteUsage()
    GradientStart=Constant() {r=-0.1}
    GradientEnd=Constant() {r=1.25}
    RotationAngle=Constant() {r=0.65}
    GradientOffset=Constant() {r=0.55}
   ) {materialFunction=&MF_UI_LinearGradient}.Result
  )
  n22 = Saturate(None=Multiply(A=Max(A=@n20, B=@n21), B=@n21))
  NamedRerouteDeclaration(None=SmoothStep(Value=Multiply(A=@n22) {constB=4}) {constMax=0.54}) {name=foreground cloud alpha_1, nodeColor=<rgba 0 1 0.8359 1>}
  NamedRerouteDeclaration(None=Constant3Vector() {constant=<rgba 0.0482 0.3419 0.7991 1>}) {name=sky gradient_1, nodeColor=<rgba 0 0.4609 1 1>}
  n23 = NamedRerouteUsage()
  n24 = Constant() {r=0.75}
  n25 = TextureSample(UVs=Panner(Coordinate=Multiply(A=@n23, B=@n24)) {speedX=0.01, speedY=0.01}) {samplerType=SAMPLERTYPE_Grayscale, texture=&LowResBlurredNoise}
  n26 = Multiply(A=@n25.R) {constB=1.5}
  n27 = Multiply(A=AppendVector(A=Constant(), B=Constant() {r=4.5}), B=NamedRerouteUsage())
  n28 = Constant()
  n29 = Constant() {r=0.2}
  n30 = Constant() {r=2}
  n31 = NamedRerouteUsage()
  n32 = Multiply(
   A=OneMinus(None=MaterialFunctionCall(UVs=@n27, Subtract=@n28, Falloff=@n29, Intensity=@n30) {materialFunction=&MF_UI_TopBottomGradient}.Result)
   B=Max(A=Multiply(A=@n31) {constB=1.25}, B=@n25.R)
  )
  n33 = SmoothStep(Value=OneMinus(None=Add(A=@n26, B=@n32))) {constMax=0.35}
  n34 = NamedRerouteUsage()
  n35 = Constant() {r=-0.01}
  n36 = Constant() {r=0.6}
  n37 = Constant() {r=0.75}
  n38 = Constant() {r=0.1}
  n39 = Multiply(
   A=@n33
   B=Min(
    A=MaterialFunctionCall(UVs=@n34, GradientStart=@n35, GradientEnd=@n36, RotationAngle=@n37, GradientOffset=@n38) {materialFunction=&MF_UI_LinearGradient}.Result
    B=@n33
   )
  )
  NamedRerouteDeclaration(None=Multiply(A=Saturate(None=@n39)) {constB=1.5}) {name=BG Cloud Alpha_1, nodeColor=<rgba 0.6641 1 0 1>}
  n40 = Constant3Vector() {constant=<rgba 0.1279 0.268 0.5833 1>}
  n41 = MaterialFunctionCall(
   A=Constant3Vector() {constant=<rgba 0.9911 0.893 0.9558 1>}
   B=Constant3Vector() {constant=<rgba 0.4007 0.375 0.9115 1>}
   C=@n40
   Alpha=NamedRerouteUsage()
  ) {materialFunction=&3ColorBlend}
  n42 = MaterialFunctionCall(
   UVs=NamedRerouteUsage()
   GradientStart=Constant() {r=0.15}
   GradientEnd=Constant() {r=1}
   RotationAngle=Constant() {r=0.75}
   GradientOffset=Constant() {r=0.25}
  ) {materialFunction=&MF_UI_LinearGradient}
  n43 = NamedRerouteUsage()
  n44 = Constant3Vector() {constant=<rgba 0.7801 0.5833 1 1>}
  n45 = Constant3Vector() {constant=<rgba 0.7969 0.4441 0.6856 1>}
  n46 = Constant3Vector() {constant=<rgba 0.2772 0.2302 0.8021 1>}
  n47 = NamedRerouteUsage()
  n48 = NamedRerouteUsage()
  n49 = Constant() {r=0.15}
  n50 = Constant() {r=1}
  n51 = Constant() {r=0.75}
  n52 = Constant() {r=0.25}
  n53 = NamedRerouteUsage()
  n54 = Constant() {r=1.25}
  n55 = TextureSample(UVs=Panner(Coordinate=Multiply(A=@n53, B=@n54)) {speedX=0.01, speedY=0.01}) {samplerType=SAMPLERTYPE_Grayscale, texture=&LowResBlurredNoise}
  n56 = Multiply(A=@n55.R) {constB=1.1}
  n57 = Multiply(A=AppendVector(A=Constant(), B=Constant() {r=5}), B=NamedRerouteUsage())
  n58 = Constant()
  n59 = Constant() {r=0.2}
  n60 = Constant() {r=2}
  n61 = Multiply(
   A=OneMinus(None=MaterialFunctionCall(UVs=@n57, Subtract=@n58, Falloff=@n59, Intensity=@n60) {materialFunction=&MF_UI_TopBottomGradient}.Result)
   B=Max(A=NamedRerouteUsage(), B=@n55.R)
  )
  n62 = SmoothStep(Value=OneMinus(None=Add(A=@n56, B=@n61))) {constMax=0.5, constMin=0.1}
  n63 = NamedRerouteUsage()
  n64 = Constant() {r=0.2}
  n65 = Constant() {r=0.65}
  n66 = Constant() {r=0.75}
  n67 = Constant() {r=0.2}
  n68 = Multiply(
   A=@n62
   B=Min(
    A=MaterialFunctionCall(UVs=@n63, GradientStart=@n64, GradientEnd=@n65, RotationAngle=@n66, GradientOffset=@n67) {materialFunction=&MF_UI_LinearGradient}.Result
    B=@n62
   )
  )
  n69 = LinearInterpolate(
   A=LinearInterpolate(A=NamedRerouteUsage(), B=LinearInterpolate(A=@n41.Result, B=@n40, Alpha=@n42.Result), Alpha=Multiply(A=@n43) {constB=1.2})
   B=LinearInterpolate(
    A=MaterialFunctionCall(A=@n44, B=@n45, C=@n46, Alpha=@n47) {materialFunction=&3ColorBlend}.Result
    B=@n46
    Alpha=MaterialFunctionCall(UVs=@n48, GradientStart=@n49, GradientEnd=@n50, RotationAngle=@n51, GradientOffset=@n52) {materialFunction=&MF_UI_LinearGradient}.Result
   )
   Alpha=Multiply(A=Saturate(None=@n68)) {constB=1.25}
  )
  n70 = NamedRerouteUsage()
  n71 = NamedRerouteUsage()
  n72 = NamedRerouteUsage()
  n73 = Constant() {r=0.05}
  n74 = Constant() {r=0.25}
  n75 = Constant() {r=0.25}
  n76 = Constant() {r=-0.07}
  NamedRerouteDeclaration(None=LinearInterpolate(
    A=Constant3Vector() {constant=<rgba 0.1894 0.0793 0.4844 1>}
    B=LinearInterpolate(A=@n69, B=@n70, Alpha=@n71)
    Alpha=MaterialFunctionCall(UVs=@n72, GradientStart=@n73, GradientEnd=@n74, RotationAngle=@n75, GradientOffset=@n76) {materialFunction=&MF_UI_LinearGradient}.Result
   )) {name=BestBG, nodeColor=<rgba 1 0.7969 0 1>}
 ]
}
